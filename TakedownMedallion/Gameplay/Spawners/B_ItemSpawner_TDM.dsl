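; Item spawner actor that places a configured item pickup in the world as floor loot, labeled as a Medallion Spawner for the TDM (Takedown Medallion) feature.
; On BeginPlay it destroys itself if its spawner hotfix toggle is off; otherwise, on the server, it waits for playlist data to be ready.
; It then spawns one pickup at its own location and prints a debug string.
; The construction script writes the item's display name onto a text render label.
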(event EventBeginPlay
  (|CallParentFunction)
  (if (ScalableFloat|GetValueasBool (Variables|Default|GetHF_SpawnerEnabled))
    (if (Networking|IsServer)
      (Class|FortAsyncActionCurrentPlaylistReady|PlaylistDataReadyAsync
        (:OnReady
          (bind _returnvalue (Inventory|SpawnItemVariantPickupinWorld (Utilities|Struct|MakeSpawnItemVariantParams (Variables|Default|GetItemDefinitionToSpawn) "(GameplayTags=)" 1 (Transformation|GetActorLocation) "0.000000,0.000000,1.000000" 0 false false false -1 "FloorLoot" "LootDrop")))
          (Development|PrintString (Utilities|String|Append "Spawned TDM: " (Utilities|GetDisplayName _returnvalue))))))
    (else
      (Actor|DestroyActor))))

(fn ConstructionScript
  (bind _itemdefinitiontospawn (Variables|Default|GetItemDefinitionToSpawn))
  (|CallParentFunction)
  (Utilities|IsValid _itemdefinitiontospawn
    (:Is%20Valid
      (Rendering|Components|TextRender|SetText (Variables|Default|GetTextRender) (Utilities|Text|FormatText "NSLOCTEXT(\"[4B61E0CFADA6F67177572532EDF6E18F]\", \"6912F23B4168AE2D066329B8A2460C0D\", \"Medallion Spawner: {Name}\")" (ItemDefinition|GetDisplayName _itemdefinitiontospawn))))))
